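# Temporary Helios landscape material function that recolors a Material input's BaseColor using four biome masks and a grass mask.
# BaseColor is fully desaturated for tinting by per-biome color and brightness parameters in the BiomeMask_Paint group.
# Chained lerps layer biomes A through D over the original color.
# The blended color apparently replaces BaseColor within the grass mask, and all other material attributes pass through.
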MaterialFunction /Game/Athena/Helios/Environments/Landscape/Materials/TEMP/MF_TempRecolor
graph {nodes [
  n1 = GetMaterialAttributes(None=FunctionInput() {inputName=Material, inputType=FunctionInput_MaterialAttributes}) {attributeGetTypes [#1]}
  n2 = NamedRerouteDeclaration(None=@n1.BaseColor) {name=InputColor, nodeColor=<rgba 0.5 1 0 1>}
  n3 = NamedRerouteUsage()
  n4 = NamedRerouteUsage()
  FunctionOutput(None=SetMaterialAttributes!(MaterialAttributes=@n1.MaterialAttributes, BaseColor=LinearInterpolate(A=@n2, B=@n3, Alpha=@n4)) {attributeSetTypes [#1]})
  NamedRerouteDeclaration(None=Desaturation(None=@n2, Fraction=Constant() {r=1})) {name=BaseColorInput, nodeColor=<rgba 0.125 0 1 1>}
  n5 = Constant()
  NamedRerouteDeclaration(None=FunctionInput(Preview=@n5) {bUsePreviewValueAsDefault=true, inputName=BiomeA, inputType=FunctionInput_Scalar}) {name=BiomeA_Mask, nodeColor=<rgba 0.9453 1 0 1>}
  NamedRerouteDeclaration(None=FunctionInput(Preview=@n5) {bUsePreviewValueAsDefault=true, inputName=BiomeB, inputType=FunctionInput_Scalar}) {name=BiomeB_Mask, nodeColor=<rgba 0.1953 0 1 1>}
  NamedRerouteDeclaration(None=FunctionInput(Preview=@n5) {bUsePreviewValueAsDefault=true, inputName=BiomeC, inputType=FunctionInput_Scalar}) {name=BiomeC_Mask, nodeColor=<rgba 1 0 0.2344 1>}
  NamedRerouteDeclaration(None=FunctionInput(Preview=@n5) {bUsePreviewValueAsDefault=true, inputName=BiomeD, inputType=FunctionInput_Scalar}) {name=BiomeD_Mask, nodeColor=<rgba 0.8516 0 1 1>}
  NamedRerouteDeclaration(None=FunctionInput(Preview=Constant()) {inputName=Grass_Mask, inputType=FunctionInput_Scalar}) {name=GrassMask, nodeColor=<rgba 0 1 0.4141 1>}
  n6 = NamedRerouteUsage()
  n7 = Multiply(
   A=VectorParameter() {group=BiomeMask_Paint, parameterName=BiomeAColor, defaultValue=<rgba 1 1 1 1>}.RGB
   B=ScalarParameter() {defaultValue=5, group=BiomeMask_Paint, parameterName=BiomeAValue}
  )
  n8 = VectorParameter() {group=BiomeMask_Paint, parameterName=BiomeBColor, defaultValue=<rgba 0.2746 1 0.7456 1>}
  n9 = ScalarParameter() {defaultValue=1, group=BiomeMask_Paint, parameterName=BiomeBValue}
  n10 = LinearInterpolate(
   A=LinearInterpolate(A=@n2, B=Multiply(A=@n6, B=@n7), Alpha=NamedRerouteUsage())
   B=Multiply(A=NamedRerouteUsage(), B=Multiply(A=@n8.RGB, B=@n9))
   Alpha=NamedRerouteUsage()
  )
  n11 = Multiply(
   A=NamedRerouteUsage()
   B=Multiply(
    A=VectorParameter() {group=BiomeMask_Paint, parameterName=BiomeCColor, defaultValue=<rgba 0.1113 0.307 0.1067 1>}.RGB
    B=ScalarParameter() {defaultValue=1, group=BiomeMask_Paint, parameterName=BiomeCValue}
   )
  )
  n12 = NamedRerouteUsage()
  n13 = NamedRerouteUsage()
  n14 = Multiply(
   A=VectorParameter() {group=BiomeMask_Paint, parameterName=BiomeDColor, defaultValue=<rgba 1 0.7589 0.2611 1>}.RGB
   B=ScalarParameter() {defaultValue=2, group=BiomeMask_Paint, parameterName=BiomeDValue}
  )
  NamedRerouteDeclaration(None=LinearInterpolate(A=LinearInterpolate(A=@n10, B=@n11, Alpha=@n12), B=Multiply(A=@n13, B=@n14), Alpha=NamedRerouteUsage())) {name=BlendedColors, nodeColor=<rgba 0 0.2031 1 1>}
 ]}
